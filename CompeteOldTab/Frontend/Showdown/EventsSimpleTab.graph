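# EventsSimpleTab is a CommonButtonLegacy tab widget in the Showdown frontend, with a skewed-box material background behind two stacked text labels.
# PreConstruct, Construct and UpdateColor apply label text, optional uppercasing, padding and Color map values, feeding a dynamic material.
# Selection changes update VisualSelectionState and, except on mobile, play Selected forward or reversed to crossfade labels and reveal the background.
# Hover scales the labels only while unselected.
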WidgetBlueprint /CompeteOldTab/Frontend/Showdown/EventsSimpleTab
animations [
 Hovered {
  binding_count=2
  length_seconds=0.1
  object=&self:Hovered
  sequence {bindings [
    CenterButtonTextWidget {
     id=#1
     display_name=CenterButtonTextWidget_White
     possessed_class=CommonTextBlock
     tracks [{
       type=MovieScene2DTransformTrack
       sections [{
         type=MovieScene2DTransformSection
         end_frame=2
         start_frame=0
         channels [
          Translation.X {type=MovieSceneScriptingFloatChannel, const [0,1]}
          Translation.Y {type=MovieSceneScriptingFloatChannel, const [0,1]}
          Angle {type=MovieSceneScriptingFloatChannel, const [0,1]}
          Scale.X {type=MovieSceneScriptingFloatChannel, keys [[0,0.875],[2,1]]}
          Scale.Y {type=MovieSceneScriptingFloatChannel, keys [[0,0.875],[2,1]]}
          Shear.X {type=MovieSceneScriptingFloatChannel, const [0,1]}
          Shear.Y {type=MovieSceneScriptingFloatChannel, const [0,1]}
         ]
        }]
      }]
    }
    CenterButtonTextWidget {
     id=#2
     possessed_class=CommonTextBlock
     tracks [{
       type=MovieScene2DTransformTrack
       sections [{
         type=MovieScene2DTransformSection
         end_frame=2
         start_frame=0
         channels [
          Translation.X {type=MovieSceneScriptingFloatChannel, const [0,1]}
          Translation.Y {type=MovieSceneScriptingFloatChannel, const [0,1]}
          Angle {type=MovieSceneScriptingFloatChannel, const [0,1]}
          Scale.X {type=MovieSceneScriptingFloatChannel, keys [[0,0.875],[2,1]]}
          Scale.Y {type=MovieSceneScriptingFloatChannel, keys [[0,0.875],[2,1]]}
          Shear.X {type=MovieSceneScriptingFloatChannel, const [0,1]}
          Shear.Y {type=MovieSceneScriptingFloatChannel, const [0,1]}
         ]
        }]
      }]
    }
   ]}
 }
 Selected {
  binding_count=3
  length_seconds=0.1
  object=&self:Selected
  sequence {bindings [
    CenterButtonTextWidget {
     id=#3
     possessed_class=CommonTextBlock
     tracks [
      {
       type=MovieSceneFloatTrack
       display_name=Render Opacity
       sections [{type=MovieSceneFloatSection, end_frame=2, start_frame=0, channels [None {type=MovieSceneScriptingFloatChannel, keys [[0,0],[1,1],[2,1]]}]}]
      }
      {
       type=MovieScene2DTransformTrack
       sections [{
         type=MovieScene2DTransformSection
         end_frame=2
         start_frame=0
         channels [
          Translation.X {type=MovieSceneScriptingFloatChannel, const [0,1]}
          Translation.Y {type=MovieSceneScriptingFloatChannel, const [0,1]}
          Angle {type=MovieSceneScriptingFloatChannel, const [0,1]}
          Scale.X {type=MovieSceneScriptingFloatChannel, keys [[0,0.875],[1,1],[2,1]]}
          Scale.Y {type=MovieSceneScriptingFloatChannel, keys [[0,0.875],[1,1],[2,1]]}
          Shear.X {type=MovieSceneScriptingFloatChannel, const [0,1]}
          Shear.Y {type=MovieSceneScriptingFloatChannel, const [0,1]}
         ]
        }]
      }
     ]
    }
    Image_ButtonTop {
     id=#4
     possessed_class=Image
     tracks [
      {
       type=MovieSceneColorTrack
       display_name=Color and Opacity
       sections [{
         type=MovieSceneColorSection
         end_frame=2
         start_frame=0
         channels [
          Color.R {type=MovieSceneScriptingFloatChannel, const [1,1]}
          Color.G {type=MovieSceneScriptingFloatChannel, const [1,1]}
          Color.B {type=MovieSceneScriptingFloatChannel, const [1,1]}
          Color.A {type=MovieSceneScriptingFloatChannel, keys [[0,0],[1,1],[2,1]]}
         ]
        }]
      }
      {
       type=MovieScene2DTransformTrack
       sections [{
         type=MovieScene2DTransformSection
         end_frame=2
         start_frame=0
         channels [
          Translation.X {type=MovieSceneScriptingFloatChannel, const [0,1]}
          Translation.Y {type=MovieSceneScriptingFloatChannel, const [0,1]}
          Angle {type=MovieSceneScriptingFloatChannel, const [0,1]}
          Scale.X {type=MovieSceneScriptingFloatChannel, const [1,1]}
          Scale.Y {type=MovieSceneScriptingFloatChannel, keys [[0,0.7],[1,0.75],[2,1]]}
          Shear.X {type=MovieSceneScriptingFloatChannel, const [0,1]}
          Shear.Y {type=MovieSceneScriptingFloatChannel, const [0,1]}
         ]
        }]
      }
     ]
    }
    CenterButtonTextWidget {
     id=#5
     display_name="Image_ButtonTop (2)"
     possessed_class=CommonTextBlock
     tracks [
      {
       type=MovieSceneFloatTrack
       display_name=Render Opacity
       sections [{type=MovieSceneFloatSection, end_frame=2, start_frame=0, channels [None {type=MovieSceneScriptingFloatChannel, keys [[0,1],[1,0],[2,0]]}]}]
      }
      {
       type=MovieScene2DTransformTrack
       sections [{
         type=MovieScene2DTransformSection
         end_frame=2
         start_frame=0
         channels [
          Translation.X {type=MovieSceneScriptingFloatChannel, const [0,1]}
          Translation.Y {type=MovieSceneScriptingFloatChannel, const [0,1]}
          Angle {type=MovieSceneScriptingFloatChannel, const [0,1]}
          Scale.X {type=MovieSceneScriptingFloatChannel, keys [[0,0.875],[1,1],[2,1]]}
          Scale.Y {type=MovieSceneScriptingFloatChannel, keys [[0,0.875],[1,1],[2,1]]}
          Shear.X {type=MovieSceneScriptingFloatChannel, const [0,1]}
          Shear.Y {type=MovieSceneScriptingFloatChannel, const [0,1]}
         ]
        }]
      }
     ]
    }
   ]}
 }
]
graphs [EventGraph {nodes [
   n1 = OnSelectedChangedEvent!() -> (OutputDelegate, then, Button:Common Button Legacy Object Reference, Selected:bool=~) {is=title K2Node_CustomEvent}
   > n2 = Branch(Condition="Equal (Boolean)"!(A:bool=@n1.Selected, B:bool=Get VisualSelectionState() -> (VisualSelectionState:bool)) -> (ReturnValue:bool))
   n3 = Set VisualSelectionState(execute=@n2.else, VisualSelectionState:bool=@n1.Selected) -> (Output_Get:bool)
   > n4 = Branch(Condition=IsMobileGame())
   n5 = Branch(execute=@n4.else, Condition=@n3)
   > PlayAnimationForward(InAnimation=Get Selected() -> (Selected:Widget Animation Object Reference))
   PlayAnimationReverse(execute=@n5.else, InAnimation=Get Selected())
   n6 = UpdateColor!() -> (OutputDelegate, then, ParameterName:String, New Color:Linear Color Structure) {is=title K2Node_CustomEvent}
   > n7 = Add!(
    execute
    TargetMap:Map of Strings to Linear Color Structures=Get Color() -> (Color:Map of Strings to Linear Color Structures)
    Key:"String (by ref)"=@n6.ParameterName
    Value:"Linear Color Structure (by ref)"=@n6.New Color
   ) -> (then) {is=title K2Node_CallFunction}
   n8 = Get ButtonNameText() -> (ButtonNameText:Text)
   n9 = Get CenterButtonTextWidget() -> (CenterButtonTextWidget:Common Text Object Reference)
   "SetText (Text)"(
    execute=[@n7.then, @n15.then, @n16.then]
    self=[@n9, Get CenterButtonTextWidget_White() -> (CenterButtonTextWidget_White:Common Text Object Reference)]
    InText=Select(Option 0:Text=@n8, Option 1:Text=TextToUpper!(InText:"Text (by ref)"=@n8) -> (ReturnValue:Text), Index=Get toUpper() -> (toUpper:bool)) -> (ReturnValue:Text)
   )
   n10 = Get Color()
   > SetColorAndOpacity(
    self=@n9
    InColorAndOpacity=Make Slate Color(SpecifiedColor=Find(TargetMap:Map of Strings to Linear Color Structures=@n10, Key=TextColor) -> (Value:Linear Color Structure))
   )
   > n11 = GetDynamicMaterial(self=Get Image_ButtonTop() -> (Image_ButtonTop:Image Object Reference))
   n12 = Get Color()
   n13 = Find(TargetMap=@n12, Key=ButtonColor1)
   > SetVectorParameterValue(self=@n11, ParameterName=Color1, Value=@n13)
   > Set GamepadButtonTextColor(GamepadButtonTextColor:Linear Color Structure=@n13) -> (Output_Get:Linear Color Structure)
   > SetVectorParameterValue(self=@n11, ParameterName=Color2, Value=Find(TargetMap=@n12, Key=ButtonColor2))
   > SetVectorParameterValue(self=@n11, ParameterName=Hover, Value=Find(TargetMap=@n12, Key=ButtonHover))
   > SetBrushFromMaterial(self=Get Image_ButtonTop(), Material=@n11)
   n14 = BreakVector2D(InVec=Get ButtonPadding() -> (ButtonPadding:Vector 2D Structure))
   > SetPadding!(
    execute
    self:Border Object Reference=Get CommonBorder_VariablePadding() -> (CommonBorder_VariablePadding:Common Border Object Reference)
    InPadding:Margin Structure=Make Margin(Left=@n14, Top=@n14.Y, Right=@n14, Bottom=@n14.Y)
   ) -> (then)
   n15 = Event PreConstruct()
   n16 = Event Construct()
   Event On Hovered()
   n17 = Get VisualSelectionState()
   > n18 = Branch(Condition=@n17)
   n19 = Get Hovered() -> (Hovered:Widget Animation Object Reference)
   PlayAnimationForward(execute=@n18.else, InAnimation=@n19, PlaybackSpeed=1.25)
   Event OnInitialized()
   > Assign On Selected Changed!(execute:Exec, self:Common Button Legacy Object Reference, Delegate:"Delegate (by ref)"=@n1) -> (then:Exec) {is=title K2Node_AssignDelegate}
   Event On Unhovered()
   > n20 = Branch(Condition=@n17)
   PlayAnimationReverse(execute=@n20.else, InAnimation=@n19, PlaybackSpeed=3)
  ]}]
variables [
 ButtonNameText {type=string, default="[ph] Tab Text"}
 ButtonPadding {type=Vector2D, default=<16 8>}
 Color {
  type="map<LinearColor>"
  default {
   ButtonColor1=<rgba 1 1 1 1>
   ButtonColor2=<rgba 1 1 1 1>
   ButtonHover=<rgba 1 1 1 1>
   TextColor=<rgba 1 1 1 1>
   TextHover=<rgba 0 0 0 1>
   TextPressed=<rgba 0 0 0 1>
   TextPressedHover=<rgba 0 0 0 1>
  }
 }
 DefaultTextColor {type=LinearColor, default=<rgba 0 0 0 0>}
 GamepadButtonTextColor {type=LinearColor, default=<rgba 0 0 0 0>}
 toUpper {type=boolean, default=true}
 VisualSelectionState {type=boolean}
 XL {type=boolean, default=true}
]
widgets {
 info {inheritedWidgetCount=1, parentClass=&CommonButtonLegacy, rootWidgetClass=&Overlay, widgetCount=9}
 widgets [
  Overlay_0 {widgetClassPath=&Overlay, widget=Overlay_0}
  SizeBox_0 {
   widgetClassPath=&SizeBox
   parent=Overlay_0
   widget=SizeBox_0
   properties {bOverride_HeightOverride=true, bOverride_WidthOverride=true}
   slot {class=OverlaySlot, properties {horizontalAlignment=HAlign_Fill, verticalAlignment=VAlign_Fill}}
  }
  Overlay_ButtonBG {widgetClassPath=&Overlay, bIsVariable=true, parent=SizeBox_0, widget=Overlay_ButtonBG, slot {class=SizeBoxSlot}}
  Image_ButtonTop {
   widgetClassPath=&Image
   bIsVariable=true
   parent=Overlay_ButtonBG
   widget=Image_ButtonTop
   properties {
    brush {
     bIsDynamicallyLoaded=false
     drawAs=Image
     imageType=NoImage
     mirroring=NoMirror
     resourceName=None
     resourceObject=/CompeteOldTab/Frontend/Showdown/Materials/M_UI_SkewBox_SButton
     tiling=NoTile
     imageSize=<256 256>
     margin=<ltrb 0.2 0.2 0.2 0.2>
     outlineSettings {
      bUseBrushTransparency=false
      roundingType=HalfHeightRadius
      width=0
      color {colorUseRule=UseColor_Specified, specifiedColor=<rgba 0 0 0 0>}
      cornerRadii=<0 0 0 0>
     }
     tintColor {colorUseRule=UseColor_Specified, specifiedColor=<rgba 1 1 1 1>}
     uVRegion {bIsValid=false, max=<0 0>, min=<0 0>}
    }
    colorAndOpacity=<rgba 1 1 1 0>
   }
   slot {class=OverlaySlot, properties {horizontalAlignment=HAlign_Fill, verticalAlignment=VAlign_Fill}}
  }
  CommonBorder_VariablePadding {
   widgetClassPath=&CommonBorder
   bIsVariable=true
   parent=Overlay_0
   widget=CommonBorder_VariablePadding
   properties {padding=<ltrb 8 8 8 8>}
   slot {class=OverlaySlot, properties {horizontalAlignment=HAlign_Fill, verticalAlignment=VAlign_Fill}}
  }
  Overlay_1 {
   widgetClassPath=&Overlay
   parent=CommonBorder_VariablePadding
   widget=Overlay_1
   slot {class=BorderSlot, properties {padding=<ltrb 8 8 8 8>}}
  }
  CenterButtonTextWidget {
   widgetClassPath=&CommonTextBlock
   bIsVariable=true
   parent=Overlay_1
   widget=CenterButtonTextWidget
   properties {
    justification=Center
    lineHeightPercentage=0.8
    text=TEXT BLOCK
    colorAndOpacity {colorUseRule=UseColor_Specified, specifiedColor=<rgba 0 0 0 1>}
    font {
     bForceMonospaced=false
     bMaterialIsStencil=false
     fontMaterial=null
     fontObject=&BurbankSmall
     letterSpacing=0
     monospacedWidth=1
     size=22
     skewAmount=0
     typefaceFontName=Black
     outlineSettings {
      bApplyOutlineToDropShadows=false
      bMiteredCorners=false
      bSeparateFillAlpha=false
      outlineMaterial=null
      outlineSize=0
      outlineColor=<rgba 0 0 0 1>
     }
    }
    renderTransform {angle=0, scale=<0.875 0.875>, shear=<0 0>, translation=<0 0>}
   }
   slot {class=OverlaySlot}
  }
  CenterButtonTextWidget_White {
   widgetClassPath=&CommonTextBlock
   bIsVariable=true
   parent=Overlay_1
   widget=CenterButtonTextWidget_White
   properties {
    justification=Center
    lineHeightPercentage=0.8
    text="[PH] TAB TEXT"
    font {
     bForceMonospaced=false
     bMaterialIsStencil=false
     fontMaterial=null
     fontObject=&BurbankSmall
     letterSpacing=0
     monospacedWidth=1
     size=22
     skewAmount=0
     typefaceFontName=Black
     outlineSettings {
      bApplyOutlineToDropShadows=false
      bMiteredCorners=false
      bSeparateFillAlpha=false
      outlineMaterial=null
      outlineSize=0
      outlineColor=<rgba 0 0 0 1>
     }
    }
    renderTransform {angle=0, scale=<0.875 0.875>, shear=<0 0>, translation=<0 0>}
   }
   slot {class=OverlaySlot}
  }
  InputActionWidget {widgetClassPath=&CommonActionWidget, bInherited=true, widget=null}
 ]
}
