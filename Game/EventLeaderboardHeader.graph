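# Widget Blueprint for the round title header of an event leaderboard, derived from FortShowdownDetailView.
# A size box holds a skew-box material border around a single title text block.
# On Refresh Data, the title takes the parent's GetWindowRoundDisplayString result.
# The same event themes it from TournamentDisplayInfo, applying PrimaryColor to the border and ShadowColor to the text, overriding the designer's placeholder text and border color.
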WidgetBlueprint /Game/EventLeaderboardHeader
graphs [EventGraph {nodes [
   Event Refresh Data() {is=title K2Node_Event}
   > "SetText (Text)"(
    self=Get Text_RoundTitle() -> (Text_RoundTitle:Common Text Object Reference)
    InText=GetWindowRoundDisplayString!(self:Fort Showdown Detail View Object Reference) -> (ReturnValue:Text)
   )
   n1 = Break Fort Tournament Display Info!(FortTournamentDisplayInfo:"Fort Tournament Display Info Structure (by ref)"=Get TournamentDisplayInfo() -> (TournamentDisplayInfo:Fort Tournament Display Info Structure)) -> (
    TitleLine1:Text
    TitleLine2:Text
    ScheduleInfo:Text
    FlavorDescription:Text
    DetailsDescription:Text
    ShortFormatTitle:Text
    LongFormatTitle:Text
    BackgroundTitle:Text
    PinScoreRequirement:int
    PinEarnedText:Text
    BaseColor:Linear Color Structure
    PrimaryColor:Linear Color Structure
    SecondaryColor:Linear Color Structure
    HighlightColor:Linear Color Structure
    TitleColor:Linear Color Structure
    ShadowColor:Linear Color Structure
    BackgroundLeftColor:Linear Color Structure
    BackgroundRightColor:Linear Color Structure
    BackgroundTextColor:Linear Color Structure
    PosterFadeColor:Linear Color Structure
    AlertText:Text
    SeriesPointLeaderboardName:Text
    AlertType:EFortTournamentAlertType Enum
    RoundDisplayNames:Array of Texts
   ) {is=title K2Node_BreakStruct}
   > SetBrushColor(self=Get CommonBorder_ScoringTitleBG() -> (CommonBorder_ScoringTitleBG:Common Border Object Reference), InBrushColor=@n1.PrimaryColor)
   > SetColorAndOpacity(self=Get Text_RoundTitle(), InColorAndOpacity=Make Slate Color(SpecifiedColor=@n1.ShadowColor))
  ]}]
variables []
widgets {
 info {inheritedWidgetCount=6, parentClass=/Script/FortniteUI.FortShowdownDetailView, rootWidgetClass=&SizeBox, widgetCount=9}
 widgets [
  SizeBox_6 {
   widgetClassPath=&SizeBox
   widget=SizeBox_6
   properties {bOverride_MinDesiredHeight=true, bOverride_MinDesiredWidth=true, minDesiredHeight=61, minDesiredWidth=260}
  }
  CommonBorder_ScoringTitleBG {
   widgetClassPath=&CommonBorder
   bIsVariable=true
   parent=SizeBox_6
   widget=CommonBorder_ScoringTitleBG
   properties {
    verticalAlignment=VAlign_Center
    background {
     bIsDynamicallyLoaded=false
     drawAs=Box
     imageType=NoImage
     mirroring=NoMirror
     resourceName=None
     resourceObject=/Game/UI/Frontend/Showdown/Materials/M_UI_SkewBox_TitleCut
     tiling=NoTile
     imageSize=<256 64>
     margin=<ltrb 0.4 0.2 0.4 0.2>
     outlineSettings {
      bUseBrushTransparency=false
      roundingType=HalfHeightRadius
      width=0
      color {colorUseRule=UseColor_Specified, specifiedColor=<rgba 0 0 0 0>}
      cornerRadii=<0 0 0 0>
     }
     tintColor {colorUseRule=UseColor_Specified, specifiedColor=<rgba 1 1 1 1>}
     uVRegion {bIsValid=false, max=<0 0>, min=<0 0>}
    }
    brushColor=<rgba 0 0.2086 1 1>
    padding=<ltrb 33 8 33 6>
   }
   slot {class=SizeBoxSlot}
  }
  Text_RoundTitle {
   widgetClassPath=&CommonTextBlock
   bIsVariable=true
   parent=CommonBorder_ScoringTitleBG
   widget=Text_RoundTitle
   properties {
    text="[ph] ]ROUND 1 - SESSION 111"
    font {
     bForceMonospaced=false
     bMaterialIsStencil=false
     fontMaterial=null
     fontObject=&Burbank
     letterSpacing=0
     monospacedWidth=1
     size=24
     skewAmount=0
     typefaceFontName=BurbankRegular-Black
     outlineSettings {
      bApplyOutlineToDropShadows=false
      bMiteredCorners=false
      bSeparateFillAlpha=false
      outlineMaterial=null
      outlineSize=0
      outlineColor=<rgba 0 0 0 1>
     }
    }
   }
   slot {class=BorderSlot, properties {verticalAlignment=VAlign_Center, padding=<ltrb 33 8 33 6>}}
  }
  Text_TimeRemaining {widgetClassPath=&CommonTextBlock, bInherited=true, widget=null}
  Text_DebugId {widgetClassPath=&TextBlock, bInherited=true, widget=null, properties=~}
  Image_PosterFront {widgetClassPath=/Script/EpicCMSUIFramework.EpicCMSImage, bInherited=true, widget=null}
  Image_PosterBack {widgetClassPath=&EpicCMSImage, bInherited=true, widget=null}
  Image_LoadingScreen {widgetClassPath=&EpicCMSImage, bInherited=true, widget=null}
  Image_PlaylistTile {widgetClassPath=&EpicCMSImage, bInherited=true, widget=null}
 ]
}
