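# List entry widget for a live tournament game session, showing game duration and player list, built on FortTournamentLiveGameSessionEntry.
# On entry set, the index parity recolors the Zebra border, giving odd rows a faint white fill and even rows none.
# Selection plays an animation that fades in a white background and darkens content to black, reversed at slower speed on deselect.
# A blue IsPlayer border wraps the row but is not driven by this graph.
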WidgetBlueprint /Game/UI/Events/TournamentLiveGamesEntry
animations [Selected {
  binding_count=1
  length_seconds=0.3
  object=&self:Selected
  sequence {bindings [CommonBorder_Selected {
     id=#1
     possessed_class=CommonBorder
     tracks [
      {
       type=MovieSceneColorTrack
       display_name=Content Color and Opacity
       sections [{
         type=MovieSceneColorSection
         end_frame=6
         start_frame=0
         channels [
          Color.R {type=MovieSceneScriptingFloatChannel, keys [[0,1],[5.9997,0]]}
          Color.G {type=MovieSceneScriptingFloatChannel, keys [[0,1],[5.9997,0]]}
          Color.B {type=MovieSceneScriptingFloatChannel, keys [[0,1],[5.9997,0]]}
          Color.A {type=MovieSceneScriptingFloatChannel, const [1,2]}
         ]
        }]
      }
      {
       type=MovieSceneColorTrack
       display_name=Brush Color
       sections [{
         type=MovieSceneColorSection
         end_frame=6
         start_frame=0
         channels [
          Color.R {type=MovieSceneScriptingFloatChannel, const [1,2]}
          Color.G {type=MovieSceneScriptingFloatChannel, const [1,2]}
          Color.B {type=MovieSceneScriptingFloatChannel, const [1,2]}
          Color.A {type=MovieSceneScriptingFloatChannel, keys [[0,0],[5.9997,1]]}
         ]
        }]
      }
     ]
    }]}
 }]
graphs [EventGraph {nodes [
   Event On Selected()
   n1 = Get Selected() -> (Selected:Widget Animation Object Reference)
   > PlayAnimation(InAnimation=@n1, PlaybackSpeed=6)
   Event On Deselected()
   > PlayAnimation(InAnimation=@n1, PlayMode=Reverse, PlaybackSpeed=3)
   n2 = Event OnEntrySet!() -> (OutputDelegate, then, EntryIndex:int=0) {is=title K2Node_Event}
   n3 = "% (Integer)"(A=@n2.EntryIndex, B=2) -> (ReturnValue=0)
   > SetBrushColor(
    self=Get CommonBorder_Zebra() -> (CommonBorder_Zebra:Common Border Object Reference)
    InBrushColor=Select(
     Option 0:Linear Color Structure=(R=1,G=1,B=1,A=0.1)
     Option 1:Linear Color Structure=(R=1,G=1,B=1,A=0)
     Index="Equal (Integer)"(A=@n3, B=0) {is=title K2Node_CallFunction}
    ) -> (ReturnValue:Linear Color Structure)
   )
  ]}]
variables []
widgets {
 info {
  inheritedWidgetCount=3
  parentClass=/Script/FortniteUI.FortTournamentLiveGameSessionEntry
  rootWidgetClass=&VerticalBox
  widgetCount=13
 }
 widgets [
  VerticalBox_2 {widgetClassPath=&VerticalBox, widget=VerticalBox_2}
  CommonBorder_IsPlayer {
   widgetClassPath=&CommonBorder
   bIsVariable=true
   parent=VerticalBox_2
   widget=CommonBorder_IsPlayer
   properties {
    background {
     bIsDynamicallyLoaded=false
     drawAs=Image
     imageType=NoImage
     mirroring=NoMirror
     resourceName=None
     resourceObject=null
     tiling=NoTile
     imageSize=<32 32>
     margin=<ltrb 0 0 0 0>
     outlineSettings {
      bUseBrushTransparency=false
      roundingType=HalfHeightRadius
      width=0
      color {colorUseRule=UseColor_Specified, specifiedColor=<rgba 0 0 0 0>}
      cornerRadii=<0 0 0 0>
     }
     tintColor {colorUseRule=UseColor_Specified, specifiedColor=<rgba 1 1 1 1>}
     uVRegion {bIsValid=false, max=<0 0>, min=<0 0>}
    }
    brushColor=<rgba 0.0009 0.1812 0.8796 0>
    padding=<ltrb 0 0 0 0>
   }
   slot {class=VerticalBoxSlot}
  }
  CommonBorder_Zebra {
   widgetClassPath=&CommonBorder
   bIsVariable=true
   parent=CommonBorder_IsPlayer
   widget=CommonBorder_Zebra
   properties {
    background {
     bIsDynamicallyLoaded=false
     drawAs=Image
     imageType=NoImage
     mirroring=NoMirror
     resourceName=None
     resourceObject=null
     tiling=NoTile
     imageSize=<32 32>
     margin=<ltrb 0 0 0 0>
     outlineSettings {
      bUseBrushTransparency=false
      roundingType=HalfHeightRadius
      width=0
      color {colorUseRule=UseColor_Specified, specifiedColor=<rgba 0 0 0 0>}
      cornerRadii=<0 0 0 0>
     }
     tintColor {colorUseRule=UseColor_Specified, specifiedColor=<rgba 1 1 1 1>}
     uVRegion {bIsValid=false, max=<0 0>, min=<0 0>}
    }
    brushColor=<rgba 0.0185 1 0.9047 0.2>
    padding=<ltrb 0 0 0 0>
   }
   slot {class=BorderSlot, properties {padding=<ltrb 0 0 0 0>}}
  }
  HorizontalBox_247 {
   widgetClassPath=&HorizontalBox
   parent=CommonBorder_Zebra
   widget=HorizontalBox_247
   slot {class=BorderSlot, properties {padding=<ltrb 0 0 0 0>}}
  }
  CommonBorder_Selected {
   widgetClassPath=&CommonBorder
   bIsVariable=true
   parent=HorizontalBox_247
   widget=CommonBorder_Selected
   properties {
    background {
     bIsDynamicallyLoaded=false
     drawAs=Image
     imageType=NoImage
     mirroring=NoMirror
     resourceName=None
     resourceObject=null
     tiling=NoTile
     imageSize=<32 32>
     margin=<ltrb 0 0 0 0>
     outlineSettings {
      bUseBrushTransparency=false
      roundingType=HalfHeightRadius
      width=0
      color {colorUseRule=UseColor_Specified, specifiedColor=<rgba 0 0 0 0>}
      cornerRadii=<0 0 0 0>
     }
     tintColor {colorUseRule=UseColor_Specified, specifiedColor=<rgba 1 1 1 1>}
     uVRegion {bIsValid=false, max=<0 0>, min=<0 0>}
    }
    brushColor=<rgba 1 1 1 0>
    padding=<ltrb 15 15 15 15>
   }
   slot {class=HorizontalBoxSlot, properties {size {sizeRule=Fill, value=1}}}
  }
  HorizontalBox_0 {
   widgetClassPath=&HorizontalBox
   parent=CommonBorder_Selected
   widget=HorizontalBox_0
   slot {class=BorderSlot, properties {padding=<ltrb 15 15 15 15>}}
  }
  SizeBox_100 {
   widgetClassPath=&SizeBox
   parent=HorizontalBox_0
   widget=SizeBox_100
   properties {bOverride_WidthOverride=true, widthOverride=150}
   slot {class=HorizontalBoxSlot, properties {horizontalAlignment=HAlign_Left, size {sizeRule=Automatic, value=0.1}}}
  }
  Text_GameTimeDuration {
   widgetClassPath=&CommonTextBlock
   bInherited=true
   bIsVariable=true
   parent=SizeBox_100
   widget=Text_GameTimeDuration
   properties {
    text=4 Seconds
    textTransformPolicy=ToUpper
    font {
     bForceMonospaced=false
     bMaterialIsStencil=false
     fontMaterial=null
     fontObject=&Burbank
     letterSpacing=0
     monospacedWidth=1
     size=24
     skewAmount=0
     typefaceFontName=Burbank-Black
     outlineSettings {
      bApplyOutlineToDropShadows=false
      bMiteredCorners=false
      bSeparateFillAlpha=false
      outlineMaterial=null
      outlineSize=0
      outlineColor=<rgba 0 0 0 1>
     }
    }
    shadowOffset=<0 0>
   }
   slot {class=SizeBoxSlot, properties {horizontalAlignment=HAlign_Left}}
  }
  CommonBorder_SelectedTint {
   widgetClassPath=&CommonBorder
   parent=HorizontalBox_0
   widget=CommonBorder_SelectedTint
   properties {padding=<ltrb 0 0 0 0>}
   slot {class=HorizontalBoxSlot, properties {verticalAlignment=VAlign_Center, padding=<ltrb 30 0 30 0>}}
  }
  HorizontalBox_208 {
   widgetClassPath=&HorizontalBox
   parent=CommonBorder_SelectedTint
   widget=HorizontalBox_208
   slot {class=BorderSlot, properties {padding=<ltrb 0 0 0 0>}}
  }
  SizeBox {
   widgetClassPath=&SizeBox
   parent=HorizontalBox_208
   widget=SizeBox
   properties {bOverride_WidthOverride=true, widthOverride=775}
   slot {class=HorizontalBoxSlot, properties {size {sizeRule=Fill, value=1}}}
  }
  Text_PlayerList {
   widgetClassPath=&CommonTextBlock
   bInherited=true
   bIsVariable=true
   parent=SizeBox
   widget=Text_PlayerList
   properties {
    scrollStyle=&TextScrollStyle-Base_C
    text="Player 1, Player 2, Player 3"
    font {
     bForceMonospaced=false
     bMaterialIsStencil=false
     fontMaterial=null
     fontObject=&Burbank
     letterSpacing=0
     monospacedWidth=1
     size=24
     skewAmount=0
     typefaceFontName=Burbank-PlayerPower
     outlineSettings {
      bApplyOutlineToDropShadows=false
      bMiteredCorners=false
      bSeparateFillAlpha=false
      outlineMaterial=null
      outlineSize=0
      outlineColor=<rgba 0 0 0 1>
     }
    }
    shadowOffset=<0 0>
   }
   slot {class=SizeBoxSlot}
  }
  InputActionWidget {widgetClassPath=&CommonActionWidget, bInherited=true, widget=null}
 ]
}
